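# Sound cue for the GoldRush buff activation, playing a random wave variant with slight modulation.
# An IsLocal branch plays local sounds unattenuated; otherwise sounds are attenuated, and an IsEnemy branch keeps enemies at full level while greatly lowering non-enemy or unset cases.
# All paths share a quality-level switch that selects either a seven-wave random pool, excluded from branch culling, or a four-wave subset.
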SoundCue /BuckSixteenGameplay/Assets/Audio/Cues/SC_GoldRush_Buff_Activate
root=Branch_0
nodes [
 {id=Attenuation_0, children [Modulator_0], properties {attenuationSettings=/BuckSixteenGameplay/Assets/Audio/Settings/Att_GoldRush_Buff}}
 {id=Branch_0, children [Modulator_0,Branch_1,Branch_1], properties {boolParameterName=IsLocal}}
 {id=Branch_1, children [Attenuation_0, Modulator_1, Modulator_1], properties {boolParameterName=IsEnemy}}
 {id=Modulator_0, children [QualityLevel_0]}
 {id=Modulator_1, children [Attenuation_0], properties {pitchMax=1, pitchMin=1, volumeMax=0.15, volumeMin=0.15}}
 {id=QualityLevel_0, children [Random_1,Random_1,Random_2,Random_2]}
 {
  id=Random_1
  children [WavePlayer_0,WavePlayer_1,WavePlayer_2,WavePlayer_6,WavePlayer_7,WavePlayer_8,WavePlayer_9]
  properties {bShouldExcludeFromBranchCulling=true, weights [1,1,1,1,1,1,1]}
 }
 {id=Random_2, children [WavePlayer_10, WavePlayer_11, WavePlayer_12, WavePlayer_13], properties {weights [1,1,1,1]}}
 {id=WavePlayer_0, properties {soundWaveAssetPtr=/BuckSixteenGameplay/Assets/Audio/Waves/GoldRush_Buff_Activate_01}}
 {id=WavePlayer_1, properties {soundWaveAssetPtr=/BuckSixteenGameplay/Assets/Audio/Waves/GoldRush_Buff_Activate_02}}
 {id=WavePlayer_10, properties {soundWaveAssetPtr=&GoldRush_Buff_Activate_01}}
 {id=WavePlayer_11, properties {soundWaveAssetPtr=&GoldRush_Buff_Activate_02}}
 {id=WavePlayer_12, properties {soundWaveAssetPtr=/BuckSixteenGameplay/Assets/Audio/Waves/GoldRush_Buff_Activate_03}}
 {id=WavePlayer_13, properties {soundWaveAssetPtr=/BuckSixteenGameplay/Assets/Audio/Waves/GoldRush_Buff_Activate_04}}
 {id=WavePlayer_2, properties {soundWaveAssetPtr=&GoldRush_Buff_Activate_03}}
 {id=WavePlayer_6, properties {soundWaveAssetPtr=&GoldRush_Buff_Activate_04}}
 {id=WavePlayer_7, properties {soundWaveAssetPtr=/BuckSixteenGameplay/Assets/Audio/Waves/GoldRush_Buff_Activate_05}}
 {id=WavePlayer_8, properties {soundWaveAssetPtr=/BuckSixteenGameplay/Assets/Audio/Waves/GoldRush_Buff_Activate_06}}
 {id=WavePlayer_9, properties {soundWaveAssetPtr=/BuckSixteenGameplay/Assets/Audio/Waves/GoldRush_Buff_Activate_07}}
]
properties {maxDistance=2650, soundClassObject=&Fort_Gadgets_Items, volumeMultiplier=0.3}
