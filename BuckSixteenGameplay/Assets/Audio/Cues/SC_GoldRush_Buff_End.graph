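# Sound cue for the end of the GoldRush buff, routed to the consumables sound class.
# IsLocal branch plays local sounds without the buff attenuation node; others pass through it, with non-enemy sounds cut sharply in volume via an IsEnemy branch.
# A quality switch picks between two mixes of quieter random main waves under full-volume B waves.
# One mix has larger pools and a pitched-up main layer; the other uses fewer variations.
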SoundCue /BuckSixteenGameplay/Assets/Audio/Cues/SC_GoldRush_Buff_End
root=Branch_0
nodes [
 {id=Attenuation_0, children [Modulator_0], properties {attenuationSettings=/BuckSixteenGameplay/Assets/Audio/Settings/Att_GoldRush_Buff}}
 {id=Branch_0, children [Modulator_0,Branch_1,Branch_1], properties {boolParameterName=IsLocal}}
 {id=Branch_1, children [Attenuation_0, Modulator_1, Modulator_1], properties {boolParameterName=IsEnemy}}
 {id=Mixer_0, children [Modulator_2,Random_4], properties {inputVolume [0.25,1]}}
 {id=Mixer_1, children [Random_3,Random_2], properties {inputVolume [0.25,1]}}
 {id=Modulator_0, children [QualityLevel_0]}
 {id=Modulator_1, children [Attenuation_0], properties {pitchMax=1, pitchMin=1, volumeMax=0.15, volumeMin=0.15}}
 {id=Modulator_2, children [Random_0], properties {pitchMax=1.5, pitchMin=1.2}}
 {id=QualityLevel_0, children [Mixer_0,Mixer_0,Mixer_1,Mixer_1]}
 {
  id=Random_0
  children [WavePlayer_3, WavePlayer_4, WavePlayer_5, WavePlayer_14, WavePlayer_15, WavePlayer_16]
  properties {bShouldExcludeFromBranchCulling=true, weights [1,1,1,1,1,1]}
 }
 {id=Random_2, children [WavePlayer_26, WavePlayer_27], properties {weights [1,1]}}
 {id=Random_3, children [WavePlayer_17, WavePlayer_18, WavePlayer_19, WavePlayer_20], properties {weights [1,1,1,1]}}
 {id=Random_4, children [WavePlayer_22, WavePlayer_23, WavePlayer_24, WavePlayer_25], properties {weights [1,1,1,1]}}
 {id=WavePlayer_14, properties {soundWaveAssetPtr=/BuckSixteenGameplay/Assets/Audio/Waves/GoldRush_Buff_End_04}}
 {id=WavePlayer_15, properties {soundWaveAssetPtr=/BuckSixteenGameplay/Assets/Audio/Waves/GoldRush_Buff_End_05}}
 {id=WavePlayer_16, properties {soundWaveAssetPtr=/BuckSixteenGameplay/Assets/Audio/Waves/GoldRush_Buff_End_06}}
 {id=WavePlayer_17, properties {soundWaveAssetPtr=/BuckSixteenGameplay/Assets/Audio/Waves/GoldRush_Buff_End_01}}
 {id=WavePlayer_18, properties {soundWaveAssetPtr=/BuckSixteenGameplay/Assets/Audio/Waves/GoldRush_Buff_End_02}}
 {id=WavePlayer_19, properties {soundWaveAssetPtr=/BuckSixteenGameplay/Assets/Audio/Waves/GoldRush_Buff_End_03}}
 {id=WavePlayer_20, properties {soundWaveAssetPtr=&GoldRush_Buff_End_04}}
 {id=WavePlayer_22, properties {soundWaveAssetPtr=/BuckSixteenGameplay/Assets/Audio/GoldRush_Buff_End_B_01}}
 {id=WavePlayer_23, properties {soundWaveAssetPtr=/BuckSixteenGameplay/Assets/Audio/GoldRush_Buff_End_B_02}}
 {id=WavePlayer_24, properties {soundWaveAssetPtr=/BuckSixteenGameplay/Assets/Audio/GoldRush_Buff_End_B_03}}
 {id=WavePlayer_25, properties {soundWaveAssetPtr=/BuckSixteenGameplay/Assets/Audio/GoldRush_Buff_End_B_04}}
 {id=WavePlayer_26, properties {soundWaveAssetPtr=&GoldRush_Buff_End_B_01}}
 {id=WavePlayer_27, properties {soundWaveAssetPtr=&GoldRush_Buff_End_B_02}}
 {id=WavePlayer_3, properties {soundWaveAssetPtr=&GoldRush_Buff_End_01}}
 {id=WavePlayer_4, properties {soundWaveAssetPtr=&GoldRush_Buff_End_02}}
 {id=WavePlayer_5, properties {soundWaveAssetPtr=&GoldRush_Buff_End_03}}
]
properties {
 attenuationSettings=&Fort_Consumable_Use_Athena_Att
 maxDistance=2650
 soundClassObject=&Fort_Consumables
 volumeMultiplier=0.5
}
