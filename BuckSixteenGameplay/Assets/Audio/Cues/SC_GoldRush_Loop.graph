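# Sound cue for the Gold Rush loop, which by its attenuation asset name appears to be the audio for a player buff.
# A Mixer plays two loop waves together, with the first at full volume and the second as a much quieter layer.
# The cue routes to the player foley sound class and uses the Gold Rush buff attenuation for distance falloff.
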SoundCue /BuckSixteenGameplay/Assets/Audio/Cues/SC_GoldRush_Loop
root=Mixer_0
nodes [
 {id=Mixer_0, children [WavePlayer_0,WavePlayer_1], properties {inputVolume [1,0.2]}}
 {id=WavePlayer_0, properties {soundWaveAssetPtr=/BuckSixteenGameplay/Assets/Audio/Waves/GoldRush_Loop_01}}
 {id=WavePlayer_1, properties {soundWaveAssetPtr=/BuckSixteenGameplay/Assets/Audio/Waves/GoldRush_Loop_02}}
]
properties {
 attenuationSettings=/BuckSixteenGameplay/Assets/Audio/Settings/Att_GoldRush_Buff
 maxDistance=2650
 soundClassObject=&Fort_Player_Foley
 volumeMultiplier=1
}
